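# Idle Niagara effect for the Sunshine pickaxe: persistent flames, a glow and drifting sparks.
# FlameTop and Flame_Front burst mesh-aligned sprites across multiple renderers, scaling height by a system timeline and tinting by time of day.
# GlowCrossPlane adds glow planes and a time-of-day-scaled light, with front-end overrides here and on the flames.
# SnowParticles, despite its name, emits warm wind-blown sparks, with spawning stopped during swings and modulated by owner speed.
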NiagaraSystem /BRCosmetics/Weapons/FORT_Melee/Pickaxe_SunShine/FX/NS_Pickaxe_Sunshine_Idle
dependencies {
 usedDataInterfaces [&NiagaraDataInterfaceArrayInt32, &NiagaraDataInterfaceCurve, &NiagaraDataInterfaceColorCurve]
 usedDynamicInputs [&FN_Select_Vector_from_Time_Of_Day]
 usedModules [
  &SystemState
  &self:SystemSpawnScript.SystemScriptSource.SystemScriptGraph.NiagaraNodeAssignment_0.SetVariables_#1
  &Timeline
 ]
}
emitters {
 FlameTop {
  emitterSpawnScript [/SetVariables(
    Emitter.AngleFade:NiagaraFloat=&FN_FrontEndFloatModifier(Gameplay=1, Front End=2)
    Emitter.BackfaceOnly=1
    Emitter.DepthFade=50
    Emitter.GlobalScale:NiagaraFloat=&FN_FrontEndFloatModifier(Gameplay=1.5, Front End=2)
    Emitter.SwitchPanDirection:NiagaraFloat=&FN_FrontEndFloatModifier(Gameplay=1, Front End=0.8)
    Emitter.SwizzleUVs:NiagaraFloat=&FN_FrontEndFloatModifier(Gameplay=1, Front End=0)
   ) {isSetParameters=true, module=SetVariables_#2}]
  emitterUpdateScript [
   &EmitterState(
    Inactive Response=::NewEnumerator1
    Life Cycle Mode=::NewEnumerator1
    Loop Behavior=::NewEnumerator1
    Loop Duration Mode=::NewEnumerator1
    UseLoopDelay=false
   )
   &SpawnBurst_Instantaneous(Use Spawn Probability=false, UseLoopCountLimit=true)
   &SpawnBurst_Instantaneous(Spawn Group=1, Use Spawn Probability=false, UseLoopCountLimit=true) {module=SpawnBurst_Instantaneous001}
  ]
  particleSpawnScript [
   &InitializeParticle(
    Color=<rgba 0.5336 0.8157 1 4>
    Lifetime=&RandomRangeFloat(Minimum=1.5, Maximum=4)
    Position Offset=<2 25.5594 50.6682>
    Position Offset Coordinate Space=::Local
    Sprite Rotation Angle=0.25
    Sprite Rotation Mode=::NewEnumerator1
    Sprite Size=<49.0038 87.3548>
    Sprite Size Mode=::NewEnumerator4
    Sprite UV Mode=::NewEnumerator4
    Sprite UV Scale:Vector2f=<1 -1>
    UsePositionOffset=true
    Write Material Random=true
   )
   &CameraOffset(Camera Offset Amount=-20)
   /SetVariables(Particles.VisibilityTag=&SelectIntFromArray(Direct Array Index=$Emitter.SpawnGroup, Int Selection Array=DataInterface() {propertyValues {IntData [0,1]}})) {isSetParameters=true, module=SetVariables_#3}
   /SetVariables(
    Particles.SpriteAlignementTop:Vector3f=&RotateVectorByQuaternion(Vector To Rotate=<0 1 0>, Quaternion=$Particles.MeshOrientation)
    Particles.SpriteFacingTop:Vector3f=&RotateVectorByQuaternion(Vector To Rotate=<0 0 1>, Quaternion=$Particles.MeshOrientation)
    Particles.SpriteTopPosition:NiagaraPosition=<0 5 40>
    Particles.SpriteTopSize:Vector2f=<25.4772 39.0996>
   ) {isSetParameters=true, module=SetVariables_#4}
  ]
  particleUpdateScript [
   &ParticleState(Kill Particles When Lifetime Has Elapsed=false)
   &ScaleSpriteSize(
    Scale Factor=&MakeVector2D(X=1, Y=&Lerp_Float(A=0.9, B=1, Alpha=$System.Timeline.InverseNormalizedTime))
    Scale Sprite Size Mode=::NewEnumerator2
   )
   &ScaleColor(
    Scale Mode=::NewEnumerator1
    Scale RGBA=&MakeVector4_FromLinearColor(LinearColor=$System.TOD Color)
    ScaleRGBA=true
   )
   &Drag(Use Linear Drag=true, Use Rotational Drag=true)
   &SolveForcesAndVelocity(Clamp Velocity=false, Limit Acceleration=false)
   &AlignSpriteToMeshOrientation(Mesh Orientation Relative Sprite Alignment Vector=<0 0 1>, SetAlignment=true, SetFacing=true)
  ]
  rendererClasses [&NiagaraSpriteRendererProperties]
  renderers [
   {rendererClass=&NiagaraSpriteRendererProperties}
   {rendererClass=&NiagaraSpriteRendererProperties, rendererIndex=1}
   {rendererClass=&NiagaraSpriteRendererProperties, rendererIndex=2}
   {rendererClass=&NiagaraSpriteRendererProperties, rendererIndex=3}
   {rendererClass=&NiagaraSpriteRendererProperties, rendererIndex=4}
   {rendererClass=&NiagaraSpriteRendererProperties, rendererIndex=5}
  ]
 }
 Flame_Front {
  emitterSpawnScript [/SetVariables(Emitter.AngleFade=1, Emitter.DepthFade=50, Emitter.TopAngleFade=5) {isSetParameters=true, module=SetVariables_#2}]
  emitterUpdateScript [
   &EmitterState(
    Inactive Response=::NewEnumerator1
    Life Cycle Mode=::NewEnumerator1
    Loop Behavior=::NewEnumerator1
    Loop Duration Mode=::NewEnumerator1
    UseLoopDelay=false
   )
   &SpawnBurst_Instantaneous(Spawn Count=2, Use Spawn Probability=false, UseLoopCountLimit=true)
   &SpawnBurst_Instantaneous(Spawn Group=1, Use Spawn Probability=false, UseLoopCountLimit=true) {module=SpawnBurst_Instantaneous001}
  ]
  particleSpawnScript [
   &InitializeParticle(
    Color=<rgba 0.5336 0.8157 1 4>
    Lifetime=&RandomRangeFloat(Minimum=1.5, Maximum=4)
    Position Offset=&FN_FrontEndVector3Modifier(Gameplay=<0 9.1357 50.9892>, Front End=<-5.4187 9.1357 50.9892>)
    Position Offset Coordinate Space=::Local
    Sprite Rotation Mode=::NewEnumerator1
    Sprite Size=&FN_FrontEndVector2Modifier(Front End:Vector2f=<58.7544 32.1909>, Gameplay:Vector2f=<56.9751 27.4605>)
    Sprite Size Mode=::NewEnumerator4
    Sprite UV Mode=::NewEnumerator4
    Sprite UV Scale:Vector2f=<1 -1>
    UsePositionOffset=true
    Write Material Random=true
   )
   &CameraOffset(Camera Offset Amount=-20)
   /SetVariables(Particles.VisibilityTag=&SelectIntFromArray(Direct Array Index=$Emitter.SpawnGroup, Int Selection Array=DataInterface() {propertyValues {IntData [0,1]}})) {isSetParameters=true, module=SetVariables_#3}
   /SetVariables(
    Particles.SpriteAlignementTop:Vector3f=&RotateVectorByQuaternion(Vector To Rotate=<0 1 0>, Quaternion=$Particles.MeshOrientation)
    Particles.SpriteFacingTop:Vector3f=&RotateVectorByQuaternion(Vector To Rotate=<1 0 0>, Quaternion=$Particles.MeshOrientation)
    Particles.SpriteTopSize:Vector2f=&Vector2DFromFloat(Value=52.668)
   ) {isSetParameters=true, module=SetVariables_#4}
  ]
  particleUpdateScript [
   &ParticleState(Kill Particles When Lifetime Has Elapsed=false)
   &ScaleSpriteSize(
    Scale Factor=&MakeVector2D(X=1, Y=&Lerp_Float(A=0.7, B=1, Alpha=$System.Timeline.InverseNormalizedTime))
    Scale Sprite Size Mode=::NewEnumerator2
   )
   &ScaleColor(
    Scale Mode=::NewEnumerator1
    Scale RGBA=&MakeVector4_FromLinearColor(LinearColor=$System.TOD Color)
    ScaleRGBA=true
   )
   &Drag(Use Linear Drag=true, Use Rotational Drag=true)
   &SolveForcesAndVelocity(Clamp Velocity=false, Limit Acceleration=false)
   &AlignSpriteToMeshOrientation(
    Mesh Orientation Relative Sprite Alignment Vector=<1 0 0>
    Mesh Orientation Relative Sprite Facing Vector:Vector3f=&Lerp_Vector(A=<0 1 0>, B=<0 0 -1>, Alpha=&MakeFloatFromInt(Int=&ReturnExecIndex()))
    SetAlignment=true
    SetFacing=true
   )
  ]
  rendererClasses [&NiagaraSpriteRendererProperties]
  renderers [
   {rendererClass=&NiagaraSpriteRendererProperties}
   {rendererClass=&NiagaraSpriteRendererProperties, rendererIndex=1}
   {rendererClass=&NiagaraSpriteRendererProperties, rendererIndex=2}
   {rendererClass=&NiagaraSpriteRendererProperties, rendererIndex=3}
   {rendererClass=&NiagaraSpriteRendererProperties, rendererIndex=4}
   {rendererClass=&NiagaraSpriteRendererProperties, rendererIndex=5}
  ]
 }
 GlowCrossPlane {
  emitterUpdateScript [
   &EmitterState(
    Inactive Response=::NewEnumerator1
    Life Cycle Mode=::NewEnumerator1
    Loop Behavior=::NewEnumerator1
    Loop Duration Mode=::NewEnumerator1
    UseLoopDelay=false
   )
   &SpawnBurst_Instantaneous(Use Spawn Probability=false, UseLoopCountLimit=true)
  ]
  particleSpawnScript [
   &InitializeParticle(
    Color=<rgba 4 0.4577 0.2047 0.5>
    Lifetime=2
    Mesh Scale=<0.5 0.5 0.5>
    Mesh Scale Mode=::NewEnumerator4
    Position Offset=<2.2553 9.2146 51.2346>
    Position Offset Coordinate Space=::Local
    Sprite Rotation Angle=0.25
    Sprite Rotation Mode=::NewEnumerator1
    Sprite Size Mode=::NewEnumerator3
    Uniform Sprite Size=107.8124
    UsePositionOffset=true
    Write Material Random=true
   )
   &SpriteFacingAndAlignment(Alignment Coordinate Space=::Local, Facing Coordinate Space=::Local, WriteAlignment=true, WriteFacing=true)
   /SetVariables(
    Particles.SideAlignement:Vector3f=&TransformVector(Vector=<0 0 1>, Destination Space=::Simulation)
    Particles.SideColor:LinearColor=&MultiplyLinearColor(Linear Color=$Particles.Color, Scale Factor=<1 1 1 1>)
    Particles.SideFacing:Vector3f=&TransformVector(Vector=<0 1 0>, Destination Space=::Simulation)
    Particles.SidePosition:NiagaraPosition=&TransformPosition(Position=<0 6.6806 48.246>, Source Space=::Local, Destination Space=::Simulation)
    Particles.SideSize:Vector2f=<57.2064 114.9615>
   ) {isSetParameters=true}
  ]
  particleUpdateScript [
   &ParticleState(Kill Particles When Lifetime Has Elapsed=false)
   &FN_Light_Attributes(
    Color=<rgba 1 0.3309 0.085 1>
    Create Color Attribute=true
    Create Exponent Attribute=true
    Create Radius Attribute=true
    Exponent=6
    Radius=80
    Scale Color=&FNTimeOfDayScalar(Night Scale=2.5, Morning Scale=1, Evening Scale=1)
   )
   &SolveForcesAndVelocity(Clamp Velocity=false, Limit Acceleration=false)
   &ScaleColor(
    Scale Mode=::NewEnumerator1
    Scale RGBA=&MakeVector4_FromLinearColor(LinearColor=&FNTimeOfDayColor(Night Color=<rgba 1 1 1 0.4>, Morning Color=<rgba 1 1 1 0.2>, Day Color=<rgba 1 1 1 1>, Evening Color=<rgba 1 1 1 1>))
    ScaleRGBA=true
   )
   /SetVariables() {enabled=false, isSetParameters=true}
   &ScaleColor(Scale Alpha=&FN_FrontEndFloatModifier(Gameplay=1, Front End=0.3), ScaleA=true, ScaleRGB=true) {module=ScaleColor001}
  ]
  rendererClasses [&NiagaraSpriteRendererProperties, &NiagaraLightRendererProperties]
  renderers [
   {rendererClass=&NiagaraSpriteRendererProperties}
   {rendererClass=&NiagaraSpriteRendererProperties, rendererIndex=1}
   {rendererClass=&NiagaraLightRendererProperties, rendererIndex=2}
  ]
 }
 SnowParticles {
  emitterUpdateScript [
   &EmitterState()
   &SpawnRate(
    Spawn Probability=&Multiply_Float(
     A=&MakeCustomFloatFromBool(Bool=$User.bisSwinging, True Float=0, False Float=1)
     B=&MakeCustomFloatFromBool(
      Bool=&SetBoolByFloatComparison(A=&VectorLength(Vector=$Engine.Owner.Velocity), B=30)
      True Float=0.2
      False Float=1
     )
    )
    SpawnRate=2
    Use Spawn Probability=true
   )
  ]
  particleSpawnScript [
   &InitializeParticle(
    Color=<rgba 0.5 0.0961 0 1>
    Lifetime=&RandomRangeFloat(Minimum=1, Maximum=2)
    Position Offset=<0 0 50>
    Position Offset Coordinate Space=::Local
    Sprite Rotation Mode=::NewEnumerator2
    Sprite Size Mode=::NewEnumerator1
    Sprite UV Mode=::NewEnumerator3
    UsePositionOffset=true
    Write Material Random=true
   )
   &PartitionParticles()
   &InheritVelocity(Module.Limit Speed=true, Velocity Limit=200)
   &ShapeLocation(Box Size=<1 1 40>, MaskSpawnGroup=false, Override Seed=false, Shape Primitive=::NewEnumerator2, Surface Only=false, UseOwnerScale=true) {module=ShapeLocation001}
   &AddVelocity(Velocity=&RandomRangeVector(Minimum=<0 -80 0>, Maximum=<0 -150 0>), Velocity Speed Scale=0.4)
  ]
  particleUpdateScript [
   &ParticleState()
   &ScaleSpriteSize(Uniform Curve Sprite Scale=DataInterface() {propertyValues {Curve {keys [
        {arriveTangent=9.0725, interpMode=RCIM_Cubic, leaveTangent=9.0725, tangentMode=RCTM_User, time=0, value=0}
        {arriveTangent=-2.5249, interpMode=RCIM_Cubic, leaveTangent=-2.5249, tangentMode=RCTM_User, time=0.1408, value=0.9847}
        {arriveTangent=-0.0405, interpMode=RCIM_Cubic, leaveTangent=-0.0405, tangentMode=RCTM_User, time=1, value=0}
       ]}}})
   &ScaleColor(
    Scale Alpha=&Multiply_Float(
     A=$System.Timeline.InverseNormalizedTime
     B=&FloatFromCurve(FloatCurve=DataInterface() {propertyValues {Curve {keys [
          {interpMode=RCIM_Cubic, tangentMode=RCTM_User, time=0, value=0}
          {interpMode=RCIM_Cubic, tangentMode=RCTM_User, time=0.5, value=1}
          {tangentMode=RCTM_User, time=1, value=0}
         ]}}})
    )
    Scale RGB=&VectorFromFloat(Value=&FNTimeOfDayScalar(Night Scale=0.8, Morning Scale=1, Evening Scale=0.8))
    ScaleA=true
    ScaleRGB=true
   ) {enabled=false}
   &WindForce(Frequency=60, Offset By Time:Vector3f=<0 0 0>, Scale=1.5, Wind Speed=<0 1 0>, Wind Speed Scale=5)
   &AerodynamicDrag(
    Aerodynamic Drag:NiagaraFloat=&RandomRangeFloat(Minimum=5, Maximum=6)
    Aerodynamic Rotational Drag=0.2
    Drag Shape Curve=DataInterface() {propertyValues {Curve {keys [
        {interpMode=RCIM_Cubic, tangentMode=RCTM_User, time=0, value=0.1}
        {arriveTangent=2.0799, interpMode=RCIM_Cubic, leaveTangent=2.0799, tangentMode=RCTM_User, time=1, value=1}
       ]}}}
    Lift Shape Curve=DataInterface() {propertyValues {Curve {keys [
        {interpMode=RCIM_Cubic, tangentMode=RCTM_User, time=0, value=0}
        {interpMode=RCIM_Cubic, tangentMode=RCTM_User, time=0.5, value=1}
        {tangentMode=RCTM_User, time=1, value=0}
       ]}}}
    Use Linear Drag=true
    Use Rotational Drag=true
   )
   &CameraOffset(Camera Offset Amount=10)
   &GravityForce(Gravity=<0 0 0>) {enabled=false}
   &InheritSourceMovement(
    Applied Force Scale=&VectorFromFloat(Value=&FloatFromCurve(FloatCurve=DataInterface() {propertyValues {Curve {keys [<at 0.7 1>,<at 1 0>]}}}))
    Applied Position Scale=&VectorFromFloat(Value=&FloatFromCurve(FloatCurve=DataInterface() {propertyValues {Curve {keys [
          {arriveTangent=0.0374, interpMode=RCIM_Cubic, leaveTangent=0.0374, tangentMode=RCTM_User, time=0.2, value=0.85}
          {arriveTangent=-0.5909, interpMode=RCIM_Cubic, leaveTangent=-0.5909, tangentMode=RCTM_User, time=1, value=0}
         ]}}}))
    ApplyForce=true
    ApplyForceFalloff=false
    ApplyPosition=true
    ApplyPositionFalloff=false
    Module.LimitForceSpeed=false
   )
   &CurlNoiseForce(Noise Strength=1, Pan Noise=false)
   &SolveForcesAndVelocity(Clamp Velocity=true, Limit Acceleration=false, Speed Limit=100)
   &ScaleSpriteSizeBySpeed(Max Scale Factor=<1 2>, Velocity Threshold=10)
   &ScaleColor(
    Linear Color Curve=DataInterface() {propertyValues {
      AlphaCurve {keys [<at 0 1>,<at 1 0>]}
      BlueCurve {keys [<at 0 0>,<at 0.15 0.3284>,<at 0.5852 0.0439>,<at 1 0>]}
      GreenCurve {keys [<at 0 2.6537>,<at 0.15 0.8878>,<at 0.5852 0.2242>,<at 1 0.0046>]}
      RedCurve {keys [<at 0 4>,<at 0.15 4>,<at 0.5852 1>,<at 1 0.5>]}
     }}
    Scale Mode=::NewEnumerator2
   ) {module=ScaleColor001}
   &ScaleColor(Scale Alpha=$System.Timeline001.InverseNormalizedTime, ScaleA=true, ScaleRGB=true) {module=ScaleColor002}
  ]
  rendererClasses [&NiagaraSpriteRendererProperties, &NiagaraMeshRendererProperties]
  renderers [
   {rendererClass=&NiagaraSpriteRendererProperties}
   {rendererClass=&NiagaraMeshRendererProperties, rendererIndex=1}
  ]
 }
}
summary {
 systemName=NS_Pickaxe_Sunshine_Idle
 emitters [
  {
   emitterName=GlowCrossPlane
   rendererClasses [&NiagaraSpriteRendererProperties, &NiagaraLightRendererProperties]
  }
  {emitterName=FlameTop, rendererClasses [&NiagaraSpriteRendererProperties]}
  {emitterName=Flame_Front, rendererClasses [&NiagaraSpriteRendererProperties]}
  {emitterName=SnowParticles, rendererClasses [&NiagaraSpriteRendererProperties, &NiagaraMeshRendererProperties]}
 ]
 userVariables [
  User.bIsFrontEnd {type=NiagaraBool}
  User.bIsFrontEndPreview {type=NiagaraBool}
  User.bisSwinging {type=NiagaraBool}
  User.FrontendSpawn {type=NiagaraBool}
  User.PlayerSpeed {type=NiagaraFloat, defaultValue=0, description=""}
  User.RibbonIDOffset {type=NiagaraInt32}
  User.SkeletalMesh {
   type=NiagaraDataInterfaceSkeletalMesh
   defaultValue {
    struct=NiagaraExt_VariableValue_DataInterface
    value {dataInterface=&self:NiagaraDataInterfaceSkeletalMesh_2}
   }
  }
  User.SnowBlue {type=LinearColor, description="", defaultValue {struct=LinearColor, value=<rgba 0.4375 1.1406 2 1>}}
  User.Variant {type=NiagaraFloat, defaultValue=0, description=""}
 ]
}
